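# Looping mono MetaSound for the ambient void sound of the Puppeteer boss's ghost follower.
# On play, an ADSR envelope starts the void wave player and multiplies its output for a faded-in attack.
# Two one-shot wave arrays are declared and filled, but nothing in the graph uses them.
# Ghost attenuation, the boss-fight sound class and a concurrency override shape playback in the world.
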MetaSoundSource /PuppeteerBoss/Audio/SFX/Ghost/MSS_Puppeteer_Follower_Amb_loop
inputs [
 OneShots_A {type=WaveAsset:Array}
 OneShots_B {type=WaveAsset:Array}
 UE.Source.OnPlay {type=Trigger}
 VoidLoop {type=WaveAsset}
]
interfaces [UE.OutputFormat.Mono:1.0, UE.Source:1.1]
outputs ["UE.OutputFormat.Mono.Audio:0" {type=Audio}]
pages [Default {
  input_defaults {
   UE.Source.OnPlay=null
   VoidLoop=/PuppeteerBoss/Audio/SFX/Ghost/wav/Puppeteer_Follower_Void_Loop
   OneShots_A [
    /PuppeteerBoss/Audio/SFX/Ghost/wav/Puppeteer_Follower_Void_Loop_OS_01
    /PuppeteerBoss/Audio/SFX/Ghost/wav/Puppeteer_Follower_Void_Loop_OS_02
    /PuppeteerBoss/Audio/SFX/Ghost/wav/Puppeteer_Follower_Void_Loop_OS_03
    /PuppeteerBoss/Audio/SFX/Ghost/wav/Puppeteer_Follower_Void_Loop_OS_04
   ]
   OneShots_B [
    /PuppeteerBoss/Audio/SFX/Ghost/wav/Puppeteer_Follower_Void_Loop_OS_05
    /PuppeteerBoss/Audio/SFX/Ghost/wav/Puppeteer_Follower_Void_Loop_OS_06
    /PuppeteerBoss/Audio/SFX/Ghost/wav/Puppeteer_Follower_Void_Loop_OS_07
    /PuppeteerBoss/Audio/SFX/Ghost/wav/Puppeteer_Follower_Void_Loop_OS_08
    /PuppeteerBoss/Audio/SFX/Ghost/wav/Puppeteer_Follower_Void_Loop_OS_09
   ]
  }
  nodes [
   InitVariable.Trigger()
   InitVariable.Trigger()
   ADSREnvelopeADSREnvelope_0 = ADSR Envelope.ADSR Envelope:Audio(Trigger Attack=^UE.Source.OnPlay) {also {defaults {Attack Time=1, Sustain Level=0.8}}}
   WavePlayer_0 = UE.Wave Player:Mono(Wave Asset=^VoidLoop, Play=@ADSREnvelopeADSREnvelope_0.On Attack Triggered) {also {defaults {Loop="<MetaSoundBuilderResult.SUCCEEDED: 0>"}}}
   Multiply_0 = UE.Multiply:Audio(
    PrimaryOperand=UE.Reroute(Value=UE.Reroute(Value=@ADSREnvelopeADSREnvelope_0.Out Envelope).Value).Value
    AdditionalOperands=@WavePlayer_0.Out Mono
   )
  ]
  links [{dst="@UE.OutputFormat.Mono.Audio:0", src=Multiply_0.Out}]
 }]
properties {
 assetImportData=&self:AssetImportData
 attenuationSettings=&/SA_PuppeteerGhost
 bLooping=true
 bOverrideConcurrency=true
 duration=10000
 maxDistance=4300
 sampleRate=48000
 soundClassObject=&Fort_BossFight
 volume=3.5
 concurrencyOverrides {maxCount=2, volumeScale=0.8}
}
