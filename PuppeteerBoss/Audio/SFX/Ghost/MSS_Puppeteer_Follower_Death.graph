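# One-shot mono MetaSound source for the death sound of a Puppeteer boss follower.
# A short trigger delay after play starts a nested random-sound player, fed by an array input whose default holds one death wave.
# The nested player supplies both the audio output and the finished trigger.
# Playback uses ghost attenuation, the boss-fight sound class and a send to distant reverb.
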MetaSoundSource /PuppeteerBoss/Audio/SFX/Ghost/MSS_Puppeteer_Follower_Death
inputs [
 Death Sound {type=WaveAsset:Array}
 UE.Source.OnPlay {type=Trigger}
]
interfaces [UE.OutputFormat.Mono:1.0, UE.Source.OneShot:1.0, UE.Source:1.1]
outputs [
 "UE.OutputFormat.Mono.Audio:0" {type=Audio}
 UE.Source.OneShot.OnFinished {type=Trigger}
]
pages [Default {
  input_defaults {UE.Source.OnPlay=null, Death Sound [/PuppeteerBoss/Audio/SFX/Ghost/wav/Puppeteer_Follower_Death]}
  nodes [
   TriggerDelay_0 = UE.Trigger Delay(In=^UE.Source.OnPlay) {also {defaults {Delay Time=0.2}}}
   MSS_PlayRandomSound_Mono_0 = &MSS_PlayRandomSound_Mono(Sounds=^Death Sound, UE.Source.OnPlay=@TriggerDelay_0.Out)
  ]
  links [
   {dst="@UE.OutputFormat.Mono.Audio:0", src=MSS_PlayRandomSound_Mono_0.UE.OutputFormat.Mono.Audio:0}
   {dst="@UE.Source.OneShot.OnFinished", src=MSS_PlayRandomSound_Mono_0.UE.Source.OneShot.OnFinished}
  ]
 }]
properties {
 assetImportData=&self:AssetImportData
 attenuationSettings=&/SA_PuppeteerGhost
 maxDistance=4300
 sampleRate=48000
 soundClassObject=&Fort_BossFight
 volume=2.5
 soundSubmixSends [{minSendLevel=0.3, sendLevel=0.3, soundSubmix=&Fortnite_Reverb_Distant}]
}
