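# A one-shot mono MetaSound source for the disappear sound of a Puppeteer boss ghost follower.
# On play, it passes an array of two disappear wave variants to the shared MSS_PlayRandomSound_Mono patch, which supplies the audio and finished trigger.
# It uses the Puppeteer ghost attenuation settings and the boss fight sound class.
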MetaSoundSource /PuppeteerBoss/Audio/SFX/Ghost/MSS_Puppeteer_Follower_Disappear
inputs [
 Dissapear sound {type=WaveAsset:Array}
 UE.Source.OnPlay {type=Trigger}
]
interfaces [UE.OutputFormat.Mono:1.0, UE.Source.OneShot:1.0, UE.Source:1.1]
outputs [
 "UE.OutputFormat.Mono.Audio:0" {type=Audio}
 UE.Source.OneShot.OnFinished {type=Trigger}
]
pages [Default {
  input_defaults {
   UE.Source.OnPlay=null
   Dissapear sound [/PuppeteerBoss/Audio/SFX/Ghost/wav/Puppeteer_Follower_Disappear_01, /PuppeteerBoss/Audio/SFX/Ghost/wav/Puppeteer_Follower_Disappear_02]
  }
  nodes [MSS_PlayRandomSound_Mono_0 = &MSS_PlayRandomSound_Mono(Sounds=^Dissapear sound, UE.Source.OnPlay=^UE.Source.OnPlay)]
  links [
   {dst="@UE.OutputFormat.Mono.Audio:0", src=MSS_PlayRandomSound_Mono_0.UE.OutputFormat.Mono.Audio:0}
   {dst="@UE.Source.OneShot.OnFinished", src=MSS_PlayRandomSound_Mono_0.UE.Source.OneShot.OnFinished}
  ]
 }]
properties {
 assetImportData=&self:AssetImportData
 attenuationSettings=&/SA_PuppeteerGhost
 maxDistance=4300
 sampleRate=48000
 soundClassObject=&Fort_BossFight
 volume=1.4
}
