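# One-shot mono MetaSound Source that plays footstep sounds for the ghost Follower in the Puppeteer boss content.
# A shared MSS_PlayRandomSound_Mono node picks from an array of ten footstep waves with pitch and volume variation, then drives the audio and finished outputs.
# It borrows third-person player footstep attenuation and the enemy foley sound class, overrides concurrency to limit overlapping steps, and sends to distant reverb.
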MetaSoundSource /PuppeteerBoss/Audio/SFX/Ghost/MSS_Puppeteer_Follower_FS
inputs [
 Footsteps {type=WaveAsset:Array}
 UE.Source.OnPlay {type=Trigger}
]
interfaces [UE.OutputFormat.Mono:1.0, UE.Source.OneShot:1.0, UE.Source:1.1]
outputs [
 "UE.OutputFormat.Mono.Audio:0" {type=Audio}
 UE.Source.OneShot.OnFinished {type=Trigger}
]
pages [Default {
  input_defaults {
   UE.Source.OnPlay=null
   Footsteps [
    /PuppeteerBoss/Audio/SFX/Ghost/wav/Puppeteer_Follower_FS_01
    /PuppeteerBoss/Audio/SFX/Ghost/wav/Puppeteer_Follower_FS_02
    /PuppeteerBoss/Audio/SFX/Ghost/wav/Puppeteer_Follower_FS_03
    /PuppeteerBoss/Audio/SFX/Ghost/wav/Puppeteer_Follower_FS_04
    /PuppeteerBoss/Audio/SFX/Ghost/wav/Puppeteer_Follower_FS_05
    /PuppeteerBoss/Audio/SFX/Ghost/wav/Puppeteer_Follower_FS_06
    /PuppeteerBoss/Audio/SFX/Ghost/wav/Puppeteer_Follower_FS_07
    /PuppeteerBoss/Audio/SFX/Ghost/wav/Puppeteer_Follower_FS_08
    /PuppeteerBoss/Audio/SFX/Ghost/wav/Puppeteer_Follower_FS_09
    /PuppeteerBoss/Audio/SFX/Ghost/wav/Puppeteer_Follower_FS_10
   ]
  }
  nodes [MSS_PlayRandomSound_Mono_0 = &MSS_PlayRandomSound_Mono(Sounds=^Footsteps, UE.Source.OnPlay=^UE.Source.OnPlay) {also {defaults {MinPitch=0.8, MinVolume=0.9}}}]
  links [
   {dst="@UE.OutputFormat.Mono.Audio:0", src=MSS_PlayRandomSound_Mono_0.UE.OutputFormat.Mono.Audio:0}
   {dst="@UE.Source.OneShot.OnFinished", src=MSS_PlayRandomSound_Mono_0.UE.Source.OneShot.OnFinished}
  ]
 }]
properties {
 assetImportData=&self:AssetImportData
 attenuationSettings=/Game/Sounds/Fort_Player/Footsteps/Settings/Fort_Player_FS_3P_Att
 bOverrideConcurrency=true
 maxDistance=3584
 sampleRate=48000
 soundClassObject=&Fort_Player_Foley_3P_Enemy_Athena
 volume=2
 concurrencyOverrides {maxCount=3, retriggerTime=0.15}
 soundSubmixSends [{minSendLevel=0.3, sendLevel=0.3, soundSubmix=&Fortnite_Reverb_Distant}]
}
