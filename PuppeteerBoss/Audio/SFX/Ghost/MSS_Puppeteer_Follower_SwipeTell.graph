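# Mono one-shot MetaSound source for the swipe attack tell of the Puppeteer boss's ghost Follower.
# On play, it passes an array of two attack-tell waves to the shared random-sound patch, which outputs the audio and the finished trigger.
# It routes to the enemy attack sound class with melee-attack attenuation and overrides concurrency to cap simultaneous instances and limit retriggering.
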MetaSoundSource /PuppeteerBoss/Audio/SFX/Ghost/MSS_Puppeteer_Follower_SwipeTell
inputs [
 Melee Attacks {type=WaveAsset:Array}
 UE.Source.OnPlay {type=Trigger}
]
interfaces [UE.OutputFormat.Mono:1.0, UE.Source.OneShot:1.0, UE.Source:1.1]
outputs [
 "UE.OutputFormat.Mono.Audio:0" {type=Audio}
 UE.Source.OneShot.OnFinished {type=Trigger}
]
pages [Default {
  input_defaults {
   UE.Source.OnPlay=null
   Melee Attacks [/PuppeteerBoss/Audio/SFX/Ghost/wav/Puppeteer_Follower_Attack_Tell_01, /PuppeteerBoss/Audio/SFX/Ghost/wav/Puppeteer_Follower_Attack_Tell_02]
  }
  nodes [MSS_PlayRandomSound_Mono_0 = &MSS_PlayRandomSound_Mono(Sounds=^Melee Attacks, UE.Source.OnPlay=^UE.Source.OnPlay) {also {defaults {MinPitch=1, MinVolume=1}}}]
  links [
   {dst="@UE.OutputFormat.Mono.Audio:0", src=MSS_PlayRandomSound_Mono_0.UE.OutputFormat.Mono.Audio:0}
   {dst="@UE.Source.OneShot.OnFinished", src=MSS_PlayRandomSound_Mono_0.UE.Source.OneShot.OnFinished}
  ]
 }]
properties {
 assetImportData=&self:AssetImportData
 attenuationSettings=&/SA_Puppeteer_Ghost_Attack_Melee
 bOverrideConcurrency=true
 maxDistance=3600
 sampleRate=48000
 soundClassObject=/Game/Sounds/FortSoundClassesAndModes/Fort_Enemy_Attack
 volume=1.275
 concurrencyOverrides {maxCount=3, retriggerTime=0.15}
}
